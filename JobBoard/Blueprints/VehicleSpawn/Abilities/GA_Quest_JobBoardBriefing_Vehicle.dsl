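; Job board briefing quest ability that tracks a spawned vehicle and fails the urgent quest if it is destroyed.
; Event activation stores the payload's target as the vehicle and binds its OnDestroyed, which broadcasts a Failed verb tagged with the job type.
; Ability end unbinds and, on authority with the vehicle undestroyed, broadcasts InstigatingPlayerCompleted only if the owner has the instigator tag, preventing duplicate military rank.
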(event Ability|EventActivateAbilityFromEvent
  (params (EventData :type Gameplay%20Event%20Data%20Structure%20%28by%20ref%29))
  (bind (_eventtag _instigator _target) (Utilities|Struct|BreakGameplayEventData EventData))
  ;[0 0 0] Activate from event - get the spawned vehicle reference from the event payload and listen for it being destroyed
  (|CallParentFunction EventData)
  (Variables|Default|SetbVehicleDestroyed false)
  (Utilities|IsValid _target
    (:Is%20Valid
      (bind _output_get (Variables|Default|SetVehicleReference _target))
      (Game|BindEventtoOnDestroyed _output_get (delegate Custom|OnVehicleDestroyed)))))

(event Custom|OnVehicleDestroyed
  (params (DestroyedActor :type Actor%20Object%20Reference))
  ;[0 0 0] Activate from event - get the spawned vehicle reference from the event payload and listen for it being destroyed
  ;;[1 1 1] When vehicle is destroyed, send a failure message to the urgent quest since it can no longer be completed with the spawned vehicle
  (Variables|Default|SetbVehicleDestroyed true)
  (bind _result (CustomVerb|AddTagsforSubject (CustomVerb|MakeCustomVerbMessage (CallFunction|GetAvatarasFortPlayerPawn) "(TagName=\"CustomVerb.JobBoard.Failed\")") "Job Type" (Variables|Default|GetJobTypeTags)))
  (CustomVerb|BroadcastCustomVerbMessage _result))

(event Ability|EventOnEndAbility
  (params (bWasCancelled :type Boolean))
  (|CallParentFunction bWasCancelled)
  (Game|UnbindEventfromOnDestroyed (Variables|Default|GetVehicleReference) (EventDispatchers|CreateEvent :!event Custom|OnVehicleDestroyed))
  ;[1 1 1] Verb event sent ONLY to briefing instigator. Added to prevent duplicate military rank in Ch6S4
  (Ability|HasAuthority
    (:True
      (bind _owningplayerpawn (Utilities|GetValidAbilityOwner)
        (:Is%20Valid
          (if (Variables|Default|GetbVehicleDestroyed)
            _
            (elif (GameplayTags|HasMatchingGameplayTag (Ability|GetFortAbilitySystemComponent _owningplayerpawn) (Variables|Default|GetInstigatorGameplayEffectTag))
              (CustomVerb|BroadcastCustomVerbMessage (CustomVerb|MakeCustomVerbMessage (CallFunction|GetAvatarasFortPlayerPawn) "(TagName=\"CustomVerb.JobBoard.InstigatingPlayerCompleted\")")))))))))
